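; Gameplay ability for the Season 20 Battle Pass "No Permit" quest, named for enabling jammers.
; On activation it calls the parent, applies the gameplay effect held in its GE variable to the owner, then immediately ends.
; The passive activation failure event only forwards to the parent, and the end-ability event is empty.
; Any jammer-specific behavior lies outside this graph, which names no specific effect.
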(event Ability|EventOnEndAbility
  (params (bWasCancelled :type Boolean)))

(event Ability|EventActivateAbility
  (|CallParentFunction)
  (Ability|ApplyGameplayEffectToOwner (Variables|Default|GetGE))
  (Ability|EndAbility))

(event Ability|EventFailedToActivatePassiveAbility
  (params (ActorInfo :type Gameplay%20Ability%20Actor%20Info%20Structure))
  (|CallParentFunction))
